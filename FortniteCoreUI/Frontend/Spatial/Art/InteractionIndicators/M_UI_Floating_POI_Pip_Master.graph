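# UI master material for a floating point-of-interest pip: a glowing, bordered diamond with a masked icon texture.
# Four rotated UV edge masks multiply into resolution-corrected diamond shapes, including a timed burst ring that periodically expands and fades from the center diamond.
# A sine offset bobs the UVs, with random-offset parameters to desync float and burst.
# IsBang blends to an icon-less alert palette; static switch IsAnimated selects animated or static output.
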Material /FortniteCoreUI/Frontend/Spatial/Art/InteractionIndicators/M_UI_Floating_POI_Pip_Master
graph {
 properties {
  bUsedWithStaticMesh=false
  blendMode=BLEND_Translucent
  materialDomain=MD_UI
  numCustomizedUVs=1
  pixelDepthOffsetMode=PDOM_Legacy
  usedShadingModels=Default Lit
 }
 nodes [
  n1 = ScalarParameter() {group=Icon, parameterName=Icon_FillAlpha}
  LinearInterpolate(A=@n1, Alpha=ScalarParameter() {group=Overall, parameterName=IsBang})
  n2 = TextureCoordinate()
  n3 = MaterialFunctionCall!(Rotation_Fraction=ScalarParameter() {defaultValue=-0.125, group=Diamond, parameterName=Diamond_Angle_Top_Left}, TexCoord=@n2) {materialFunction=/FortUILibrary/Foundation/Art_ToolKit/AngledBoxes/MF_UI_Rotator_Fractions}
  n4 = Multiply(A=ScalarParameter() {defaultValue=3.125, group=Diamond, parameterName=Diamond_Scale_Border}) {constB=-0.1}
  n5 = Add(A=@n3.Result, B=@n4)
  n6 = Multiply(
   A=MaterialFunctionCall!(Native_U_PixelSize=ScalarParameter() {defaultValue=80, group=Overall, parameterName=NativeUIPixelSize}) {materialFunction=/FortUILibrary/Foundation/Art_ToolKit/Tools/M_UI_Resolution_Sharpness_Correction}.Result
   B=ScalarParameter() {defaultValue=80, group=Diamond, parameterName=Diamond_Softness}
  )
  n7 = Saturate(None=Multiply(A=ComponentMask(None=@n5) {r=true}, B=@n6))
  n8 = MaterialFunctionCall(
   Rotation_Fraction=ScalarParameter() {defaultValue=-0.125, group=Diamond, parameterName=Diamond_Angle_Bottom_Right}
   TexCoord=OneMinus(None=@n2)
  ) {materialFunction=&MF_UI_Rotator_Fractions}
  n9 = Add(A=@n8.Result, B=@n4)
  n10 = Saturate(None=Multiply(A=ComponentMask(None=@n9) {r=true}, B=@n6))
  n11 = MaterialFunctionCall(Rotation_Fraction=ScalarParameter() {defaultValue=-0.375, group=Diamond, parameterName=Diamond_Angle_Top_Right}, TexCoord=@n2) {materialFunction=&MF_UI_Rotator_Fractions}
  n12 = Add(A=@n11.Result, B=@n4)
  n13 = Saturate(None=Multiply(A=ComponentMask(None=@n12) {r=true}, B=@n6))
  n14 = MaterialFunctionCall(
   Rotation_Fraction=ScalarParameter() {defaultValue=-0.375, group=Diamond, parameterName=Diamond_Angle_Bottom_Left}
   TexCoord=OneMinus(None=@n2)
  ) {materialFunction=&MF_UI_Rotator_Fractions}
  n15 = Add(A=@n14.Result, B=@n4)
  n16 = Saturate(None=Multiply(A=ComponentMask(None=@n15) {r=true}, B=@n6))
  n17 = Multiply(A=DotProduct(A=DotProduct(A=@n7, B=@n10), B=DotProduct(A=@n13, B=@n16)), B=ScalarParameter() {defaultValue=0.5, parameterName=BorderAlpha})
  n18 = MaterialFunctionCall() {materialFunction=&RadialGradientExponential}
  n19 = Multiply(
   A=Saturate(None=Power(Base=@n18.RadialGradientExponential) {constExponent=2.2})
   B=ScalarParameter() {defaultValue=0.2, parameterName=GlowAmount}
  )
  n20 = Multiply(
   A=Add(A=Time(), B=ScalarParameter() {group=Overall, parameterName=RandomOffsetValue_Burst})
   B=ScalarParameter() {defaultValue=1, group=Overall, parameterName=Overall_TimeControl}
  )
  n21 = Fmod(A=@n20, B=Constant() {r=4})
  n22 = Constant() {r=1}
  n23 = If(A=@n21, B=@n22, "A > B"=Constant(), "A == B"=@n22, "A < B"=@n21)
  n24 = Fmod(A=@n20, B=Constant() {r=1})
  n25 = DotProduct(A=Ceil(None=Saturate(None=@n23)), B=OneMinus(None=@n24))
  n26 = LinearInterpolate(B=ScalarParameter() {defaultValue=0.5, group=Burst, parameterName=Burst_Starting_Alpha}, Alpha=@n25) {constB=0.5}
  n27 = Power(Base=Logarithm10!(X=LinearInterpolate(Alpha=@n24) {constB=10}), Exp=LinearInterpolate(Alpha=@n24) {constA=0.25, constB=0.75}) {constExponent=0.5}
  n28 = Multiply(A=LinearInterpolate(
    A=ScalarParameter() {defaultValue=3.5, group=Diamond, parameterName=Diamond_Scale}
    B=ScalarParameter() {defaultValue=1.75, group=Burst, parameterName=Diamond_Scale_Burst}
    Alpha=Saturate(None=@n27)
   ) {constA=0.5}) {constB=-0.1}
  n29 = Add(A=@n3.Result, B=@n28)
  n30 = Saturate(None=Multiply(A=ComponentMask(None=@n29) {r=true}, B=@n6))
  n31 = Add(A=@n8.Result, B=@n28)
  n32 = Saturate(None=Multiply(A=ComponentMask(None=@n31) {r=true}, B=@n6))
  n33 = Add(A=@n11.Result, B=@n28)
  n34 = Saturate(None=Multiply(A=ComponentMask(None=@n33) {r=true}, B=@n6))
  n35 = Add(A=@n14.Result, B=@n28)
  n36 = Saturate(None=Multiply(A=ComponentMask(None=@n35) {r=true}, B=@n6))
  n37 = Multiply(A=ScalarParameter() {defaultValue=3.5, group=Diamond, parameterName=Diamond_Scale}) {constB=-0.1}
  n38 = Multiply(A=ComponentMask(None=Add(A=@n3.Result, B=@n37)) {r=true}, B=@n6)
  n39 = Multiply(A=ComponentMask(None=Add(A=@n8.Result, B=@n37)) {r=true}, B=@n6)
  n40 = Multiply(A=ComponentMask(None=Add(A=@n11.Result, B=@n37)) {r=true}, B=@n6)
  n41 = Multiply(A=ComponentMask(None=Add(A=@n14.Result, B=@n37)) {r=true}, B=@n6)
  n42 = DotProduct(A=DotProduct(A=Saturate(None=@n38), B=Saturate(None=@n39)), B=DotProduct(A=Saturate(None=@n40), B=Saturate(None=@n41)))
  n43 = DotProduct(A=DotProduct(A=DotProduct(A=@n30, B=@n32), B=DotProduct(A=@n34, B=@n36)), B=OneMinus(None=@n42))
  n44 = DotProduct(A=@n26, B=@n43)
  n45 = Multiply(A=@n42, B=ScalarParameter() {defaultValue=1, group=Diamond, parameterName=DiamondShape_Alpha})
  n46 = Saturate(None=Add(A=Saturate(None=@n44), B=@n45))
  n47 = TextureCoordinate()
  n48 = AppendVector(A=Constant(), B=ScalarParameter() {group=Icon, parameterName=Texture_Translate_V})
  n49 = TextureSampleParameter2D(UVs=MaterialFunctionCall(UVs=Add(A=@n47, B=@n48), Texture Scale=ScalarParameter() {defaultValue=0.3, group=Icon, parameterName=Texture_Scale}) {materialFunction=&ScaleUVsByCenter}.UVs) {
   group=Icon
   parameterName=Texture
   samplerSource=SSM_Clamp_WorldGroupSettings
   samplerType=SAMPLERTYPE_LinearColor
   texture=&/T_UI_BP12_POI_Icon_64x
  }
  n50 = Saturate(None=Multiply(
    A=StaticComponentMaskParameter(None=@n49.RGB) {defaultR=true, group=Icon, parameterName=Channel}
    B=ScalarParameter() {defaultValue=1, group=Icon, parameterName=Texture_Softness}
   ))
  n51 = Saturate(None=Add(A=@n43, B=@n50))
  n52 = DotProduct(A=@n45, B=OneMinus(None=@n50))
  n53 = Multiply(A=@n50, B=@n1)
  n54 = Add(A=@n52, B=@n53)
  n55 = Saturate(None=Add(A=DotProduct(A=@n26, B=@n51), B=@n54))
  n56 = Add(
   A=Saturate(None=Add(A=@n17, B=@n19))
   B=LinearInterpolate(
    A=StaticSwitchParameter(True=@n55, False=@n54) {defaultValue=true, parameterName=IsAnimated}
    B=@n46
    Alpha=ScalarParameter() {group=Overall, parameterName=IsBang}
   )
  )
  n57 = ScalarParameter() {defaultValue=1, group=Overall, parameterName=Overall_Alpha}
  Multiply(A=Saturate(None=Multiply(A=@n56, B=@n57)), B=ScalarParameter() {defaultValue=1, group=Overall, parameterName=Final_Alpha})
  n58 = TextureCoordinate()
  n59 = Add(A=Time(), B=ScalarParameter() {group=Overall, parameterName=RandomOffsetValue_Float})
  n60 = ScalarParameter() {defaultValue=1, group=Overall, parameterName=Overall_TimeControl}
  n61 = Sine(None=Multiply(A=Multiply(A=@n59, B=@n60), B=ScalarParameter() {defaultValue=0.125, group=Float, parameterName=Float_Speed}))
  n62 = AppendVector(
   A=Constant()
   B=LinearInterpolate(
    A=ScalarParameter() {defaultValue=-0.02, group=Float, parameterName=Float_Offset_Min}
    B=ScalarParameter() {defaultValue=0.02, group=Float, parameterName=Float_Offset_Max}
    Alpha=Add(A=@n61)
   )
  )
  Reroute(None=StaticSwitchParameter(True=Add(A=@n58, B=@n62), False=@n58) {defaultValue=true, parameterName=IsAnimated})
  n63 = VectorParameter() {group=Burst, parameterName=Diamond_Burst_Color, defaultValue=<rgba 0 0.1667 1 1>}
  n64 = VectorParameter() {group=Diamond, parameterName=Diamond_Color, defaultValue=<rgba 0 0.6667 1 1>}
  n65 = LinearInterpolate(B=ScalarParameter() {defaultValue=0.25, group=Burst, parameterName=Burst_Starting_Color_Lift}, Alpha=@n25) {constAlpha=0.2, constB=0.25}
  n66 = Saturate(None=Add(A=DotProduct(A=@n65, B=@n51), B=@n52))
  n67 = VectorParameter() {group=Burst, parameterName=Bang_Burst_Color, defaultValue=<rgba 1 0.4167 0 1>}
  n68 = LinearInterpolate(A=Constant(), B=VectorParameter() {group=Diamond, parameterName=Bang_Color, defaultValue=<rgba 1 1 0 1>}.RGB, Alpha=@n66)
  StaticSwitchParameter(
   True=LinearInterpolate(
    A=LinearInterpolate(A=@n63.RGB, B=@n64.RGB, Alpha=@n66)
    B=LinearInterpolate(A=@n67.RGB, B=@n68, Alpha=@n46)
    Alpha=ScalarParameter() {group=Overall, parameterName=IsBang}
   )
   False=LinearInterpolate(A=@n64.RGB, B=VectorParameter() {group=Icon, parameterName=IconColor, defaultValue=<rgba 1 1 1 1>}.RGB, Alpha=@n53)
  ) {defaultValue=true, parameterName=IsAnimated}
 ]
}
