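# Translucent UI material for a point-of-interest icon among the frontend spatial interaction indicators, named as a faction variant.
# It holds a linear-color sample of the T_UI_BP12_POI_Icon_2_64x texture and an RGB01 vector parameter defaulting to light gray.
# No links between these two nodes or to the material outputs are shown.
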Material /FortniteCoreUI/Frontend/Spatial/Art/InteractionIndicators/T_UI_BP12_POI_Icon_2_64x_Faction
graph {
 properties {
  bUsedWithStaticMesh=false
  blendMode=BLEND_Translucent
  materialDomain=MD_UI
  pixelDepthOffsetMode=PDOM_Legacy
  usedShadingModels=Default Lit
 }
 nodes [
  TextureSample() {samplerType=SAMPLERTYPE_LinearColor, texture=&/T_UI_BP12_POI_Icon_2_64x}
  VectorParameter() {parameterName=RGB01, defaultValue=<rgba 0.75 0.75 0.75 0.75>}
 ]
}
